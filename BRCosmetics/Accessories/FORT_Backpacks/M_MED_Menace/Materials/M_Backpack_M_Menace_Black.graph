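# Translucent skeletal-mesh material for the black Menace backpack cosmetic, layered between the shared character PreFX and PostFX functions.
# An FX texture mask separates gems from the outer shell, which gets a tinted color and rim-based opacity gated by vertex red.
# Vertex green blends in star sprites, expanded toward the camera through world position offset.
# Panned noise through a curve atlas, plus event audio amplitude, drives star twinkle, size and glow.
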Material /BRCosmetics/Accessories/FORT_Backpacks/M_MED_Menace/Materials/M_Backpack_M_Menace_Black
graph {
 properties {
  bUseMaterialAttributes=true
  bUsedWithMeshDeformer=true
  bUsedWithSkeletalMesh=true
  blendMode=BLEND_Translucent
  opacityMaskClipValue=0.3
  pixelDepthOffsetMode=PDOM_Legacy
  refractionMethod=RM_IndexOfRefraction
  usedShadingModels=Default Lit
 }
 nodes [
  n1 = MaterialFunctionCall() {materialFunction=&BaseFN_CharacterFunction_PreFX}
  n2 = GetMaterialAttributes(None=MaterialFunctionCall(
    PreskinnedPosition=@n1.Pre skinned Local Position
    Gmap=@n1.Pre skinned Local Position
    MaterialAttributes=@n1.Pre skinned Local Position
    AO=@n1.Pre skinned Local Position
   ) {materialFunction=&CharacterSkinFX}.Result) {attributeGetTypes [#1, #2]}
  n3 = TextureSample() {samplerType=SAMPLERTYPE_Masks, texture=/BRCosmetics/Accessories/FORT_Backpacks/M_MED_Menace/Textures/T_Backpack_Menace_Backpack_FX}
  n4 = LinearInterpolate(
   A=Multiply(
    A=VectorParameter() {group=Menace, parameterName=OuterColor, defaultValue=<rgba 0.9061 1 0.5289 0>}.RGB
    B=ScalarParameter() {defaultValue=1, group=Menace, parameterName=OuterColorMultiply}
   ) {constB=3}
   B=@n2.Base Color
   Alpha=@n3.R
  ) {constB=0, desc=Mask Based On Gems}
  n5 = DotProduct(A=PixelNormalWS(), B=NamedRerouteUsage())
  n6 = Power(Base=OneMinus(None=Saturate(None=@n5)), Exp=ScalarParameter() {defaultValue=1, group=Menace, parameterName=OuterWhitePower})
  n7 = ScalarParameter() {defaultValue=1, group=Menace, parameterName=OuterWhiteRegion}
  n8 = Add(B=LinearInterpolate(A=Multiply(A=@n6, B=@n7) {constB=10}, Alpha=VertexColor().R) {constB=0})
  n9 = SetMaterialAttributes!(
   MaterialAttributes=GetMaterialAttributes(None=MaterialFunctionCall!(In=@n1.Result) {materialFunction=/Game/Characters/Player/Female/Medium/Bodies/F_MED_Stars/Materials/MF_stars_black}.Result).MaterialAttributes
   Emissive Color
   Base Color=@n4
   Opacity=@n8
  ) {attributeSetTypes [#2, #1, #3]}
  n10 = SetMaterialAttributes(
   MaterialAttributes=@n2.MaterialAttributes
   Base Color=@n4
   Opacity=LinearInterpolate(A=@n8, Alpha=@n3.R) {desc=Apply Opacity to Only Non-Gems}
  ) {attributeSetTypes [#2, #1, #3]}
  n11 = VectorParameter() {group=Menace, parameterName=StarColor, defaultValue=<rgba 0.9199 0.5354 1 0>}
  n12 = CollectionParameter() {collection=&FortniteMaterialParameters, parameterName=EventAudio_Amplitude}
  n13 = MaterialFunctionCall!(Base=Multiply(A=@n11.RGB) {constB=3}, Add=Power(Base=@n12) {constExponent=1}) {materialFunction=/Engine/Functions/Engine_MaterialFunctions02/Math/MultiplyAdd}
  n14 = TextureCoordinate()
  n15 = TextureSample(UVs=@n14) {texture=/Game/Effects/Fort_Effects/Textures/Flares/T_Flare_Spikeball}
  n16 = TextureSample(UVs=@n14) {texture=/Game/Characters/Player/Female/Medium/Bodies/F_MED_Stars/Textures/starFilter}
  n17 = DotProduct(A=CameraVectorWS(), B=PixelNormalWS())
  n18 = Power(Base=@n17) {constExponent=3}
  n19 = Panner(Coordinate=VertexColor().R, Speed=ScalarParameter() {defaultValue=0.06, parameterName=TwinkleSpeed})
  n20 = CurveAtlasRowParameter(CurveTime=Power(Base=TextureSample(UVs=@n19) {samplerType=SAMPLERTYPE_LinearColor, texture=&HighContrastNoise}.R)) {atlas=&AthenaCurveAtlas01, curve=/Game/Athena/LinearColorCurves/Curve_Galileo_Hologram, defaultValue=171}
  n21 = CollectionParameter() {collection=&FortniteMaterialParameters, parameterName=EventAudio_Amplitude}
  n22 = MaterialFunctionCall(Base=@n20.B, Add=Power(Base=@n21) {constExponent=4}) {materialFunction=&MultiplyAdd}
  n23 = Multiply(A=Multiply(A=Saturate(None=@n18), B=Saturate(None=@n22.Result))) {constB=2}
  n24 = RayTracingQualitySwitch(Normal=Multiply(A=LinearInterpolate(A=@n15.R, B=@n16.R, Alpha=@n23)), RayTraced=Constant())
  n25 = MaterialFunctionCall(Float2=Subtract(A=TextureCoordinate()) {constB=0.5}) {materialFunction=&BreakOutFloat2Components}
  n26 = Multiply(A=@n25.R, B=Transform(None=Constant3Vector() {constant=<rgba 1 0 0 0>}) {transformSourceType=TRANSFORMSOURCE_Camera})
  n27 = ScalarParameter() {defaultValue=10, group=Stars, parameterName=Stars_starSizeMult}
  n28 = MaterialFunctionCall(Float2=Add(A=ScalarParameter() {group=Stars, parameterName=Stars_starMinSize}, B=Multiply(A=@n22.Result, B=@n27) {constB=10})) {materialFunction=&BreakOutFloat2Components}
  n29 = Multiply(A=@n25.G, B=Transform(None=Constant3Vector() {constant=<rgba 0 -1 0 0>}) {transformSourceType=TRANSFORMSOURCE_Camera})
  n30 = Add(
   A=Add(A=Multiply(A=@n26, B=@n28.R), B=Multiply(A=@n29, B=@n28.G))
   B=MaterialFunctionCall(Offset Amount=ScalarParameter() {defaultValue=3, group=Stars, parameterName="Stars Offset Amount (S)"}) {materialFunction=/Engine/Functions/Engine_MaterialFunctions02/WorldPositionOffset/CameraOffset}.Clamped Camera Offset
  )
  MaterialFunctionCall(
   AO=@n1.AO
   Distance Blend=@n1.AO
   Fresnel=@n1.AO
   Material=BlendMaterialAttributes(
    A=BlendMaterialAttributes(A=@n9, B=@n10, Alpha=@n3.R)
    B=MakeMaterialAttributes!(
     BaseColor=@n11.RGB
     Metallic
     Specular
     Roughness
     Anisotropy
     EmissiveColor=@n13.Result
     Opacity=@n24
     OpacityMask=@n24
     Normal
     Tangent
     WorldPositionOffset=@n30
     SubsurfaceColor
     ClearCoat
     ClearCoatRoughness
     AmbientOcclusion
     Refraction
     CustomizedUVs_0
     CustomizedUVs_1
     CustomizedUVs_2
     CustomizedUVs_3
     CustomizedUVs_4
     CustomizedUVs_5
     CustomizedUVs_6
     CustomizedUVs_7
     PixelDepthOffset
     ShadingModel
     Displacement
    )
    Alpha=VertexColor().G
   )
  ) {materialFunction=&BaseFN_CharacterFunction_PostFX}
  n31 = CameraPositionWS()
  n32 = WorldPosition()
  NamedRerouteDeclaration(None=Normalize(VectorInput=Subtract(A=@n31, B=@n32.XYZ))) {name=CameraDirection, nodeColor=<rgba 0 0.0859 1 1>}
  n33 = Subtract(A=CameraPositionWS(), B=WorldPosition().XYZ)
  n34 = Saturate(None=DotProduct(A=PixelNormalWS(), B=Normalize(VectorInput=@n33)))
  n35 = ScalarParameter() {defaultValue=1, parameterName=InnerBlackPower}
  NamedRerouteDeclaration(None=Multiply(A=Power(Base=@n34, Exp=@n35), B=ScalarParameter() {defaultValue=1, parameterName=InnerBlackRegion}) {constB=10}) {name=Inner Region, nodeColor=<rgba 0.9219 1 0 1>}
  OneMinus(None=@n17)
 ]
}
