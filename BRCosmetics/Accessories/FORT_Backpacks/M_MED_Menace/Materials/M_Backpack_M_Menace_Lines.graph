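# Character material for the glowing lines on the Menace backpack, wrapping the shared stars-lines function between character PreFX and PostFX functions.
# Panning 3D noise on pre-skinned positions drives brightness, with a powered overdrive term for hot spots.
# That brightness tints LineColorFront for base color and emissive, with emissive dimmed where vertex normals turn from the object-to-camera direction.
# Sine pulse, UV cosine mask and side/back color nodes are unconnected.
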Material /BRCosmetics/Accessories/FORT_Backpacks/M_MED_Menace/Materials/M_Backpack_M_Menace_Lines
graph {
 properties {
  bTangentSpaceNormal=false
  bUseMaterialAttributes=true
  bUsedWithMeshDeformer=true
  bUsedWithSkeletalMesh=true
  numCustomizedUVs=2
  pixelDepthOffsetMode=PDOM_Legacy
  usedShadingModels=Default Lit
 }
 nodes [
  Constant3Vector() {constant=<rgba 0.5 0.5 0.5 0>}
  n1 = ObjectPositionWS()
  n2 = CameraPositionWS()
  n3 = DotProduct(A=Subtract(A=WorldPosition().XYZ, B=@n1), B=Normalize(VectorInput=Subtract(A=@n2, B=@n1)))
  n4 = ScalarParameter() {defaultValue=0.1, group=Ribbons, parameterName=MinimumBrightness}
  n5 = ScalarParameter() {defaultValue=20, group=Ribbons, parameterName=MaximumBrightness}
  n6 = NamedRerouteUsage()
  n7 = Saturate(None=Add(A=@n6, B=ScalarParameter() {defaultValue=0.2, group=Ribbons, parameterName=PanningMinMult, sliderMax=1}) {constB=0})
  n8 = Add(A=@n6, B=ScalarParameter() {group=Ribbons, parameterName=ShiftOverdriveBasevalue})
  n9 = Multiply(
   A=Power(Base=Saturate(None=@n8), Exp=ScalarParameter() {defaultValue=20, group=Ribbons, parameterName=OverDrivePower}) {constExponent=20}
   B=ScalarParameter() {defaultValue=30, group=Ribbons, parameterName=OverDriveBrightness}
  )
  n10 = Add(A=@n4, B=Multiply(A=Subtract(A=@n5, B=@n4), B=Add(A=@n7, B=@n9)))
  n11 = Multiply(A=VectorParameter() {group=Ribbons, parameterName=LineColorFront, defaultValue=<rgba 0.4851 0.4678 0.3419 1>}.RGB, B=@n10)
  LinearInterpolate(B=@n11, Alpha=Saturate(None=Abs(None=@n3)))
  n12 = MaterialFunctionCall() {materialFunction=&BaseFN_CharacterFunction_PreFX}
  n13 = VertexNormalWS()
  n14 = Normalize(VectorInput=Subtract(A=CameraPositionWS(), B=ObjectPositionWS()))
  n15 = Saturate(None=VertexInterpolator(VS=DotProduct(A=@n13, B=@n14)).PS)
  n16 = LinearInterpolate(A=Divide(A=@n11) {constB=4}, B=@n11, Alpha=@n15) {constAlpha=0}
  n17 = SetMaterialAttributes!(
   MaterialAttributes=GetMaterialAttributes(None=MaterialFunctionCall!(In=@n12.Result) {materialFunction=/Game/Characters/Player/Female/Medium/Bodies/F_MED_Stars/Materials/MF_stars_lines}.Result) {attributeGetTypes [#1]}.MaterialAttributes
   Emissive Color=@n16
   Base Color=@n11
   Opacity
   Opacity Mask
  ) {attributeSetTypes [
    #1
    #2
    #3
    #4
   ]}
  MaterialFunctionCall(AO=@n12.AO, Distance Blend=@n12.AO, Fresnel=@n12.AO, Material=@n17) {materialFunction=&BaseFN_CharacterFunction_PostFX}
  Multiply(A=VectorParameter() {parameterName=vvvvvvvvvvvvvvvvv, defaultValue=<rgba 0 0.0081 1 0>}.RGB, B=@n10)
  n18 = Time()
  n19 = ScalarParameter() {group=Ribbons, parameterName=PulseTiming}
  n20 = Add(A=@n19)
  n21 = Add(A=@n19) {constB=5}
  n22 = Add(A=Sine(None=Divide(A=@n18, B=@n20) {constB=3}), B=Sine(None=Divide(A=@n18, B=@n21) {constB=15}))
  n23 = Sine(None=Divide(A=@n18, B=Add(A=@n19) {constB=11}) {constB=21})
  n24 = Divide(A=Add(A=Add(A=@n22, B=@n23)) {constB=3}) {constB=6}
  n25 = ScalarParameter() {defaultValue=0.3, group=Ribbons, parameterName=MinBrightness, sliderMax=1}
  Multiply(A=@n16, B=Saturate(None=Add(A=@n24, B=@n25)))
  Multiply(A=VectorParameter() {parameterName=vvvvv, defaultValue=<rgba 0 1 0.0284 0>}.RGB, B=@n10)
  n26 = TextureCoordinate()
  n27 = ComponentMask(None=@n26) {g=true}
  n28 = Cosine(None=@n27) {period=2}
  n29 = Cosine(None=ComponentMask(None=@n26) {r=true}) {period=2}
  Multiply(
   A=Max(A=Abs(None=@n28), B=ScalarParameter() {defaultValue=0.1, group=Ribbons, parameterName=MinDarknessForLine, sliderMax=1}) {constB=0.1}
   B=OneMinus(None=Abs(None=@n29))
  )
  n30 = TextureObjectParameter() {parameterName=Noise, samplerType=SAMPLERTYPE_Masks, texture=/Game/Global/Textures/Default/Noise/NoisePattern}
  n31 = ScalarParameter() {defaultValue=300, group=Ribbons, parameterName=WorldAlignedTextureSize}
  n32 = MaterialFunctionCall(X=@n31, Y=@n31, Z=@n31) {materialFunction=&MakeFloat3}
  n33 = Time()
  n34 = ScalarParameter() {defaultValue=3, group=Ribbons, parameterName=PanningSpeed}
  n35 = VertexInterpolator(VS=Add(A=PreSkinnedPosition(), B=Multiply(A=@n33, B=@n34)))
  n36 = VertexInterpolator(VS=PreSkinnedNormal())
  NamedRerouteDeclaration(None=ComponentMask(None=MaterialFunctionCall(TextureObject=@n30, TextureSize=@n32.Result, WorldPosition=@n35.PS, World Space Normal=@n36.PS) {materialFunction=&WorldAlignedTexture}.XYZ Texture) {b=true}) {name=3DNoise, nodeColor=<rgba 0 1 0.6484 1>}
  Reroute(None=@n11)
  Saturate(None=@n15)
  Sine(None=@n27) {period=2}
  Subtract(B=@n15)
  VectorParameter() {group=Ribbons, parameterName=LineColorSide, defaultValue=<rgba 0.4851 0.4678 0.3419 1>}
  VectorParameter() {group=Ribbons, parameterName=LineColorBack, defaultValue=<rgba 0.4851 0.4678 0.3419 1>}
  VectorParameter() {parameterName=vv, defaultValue=<rgba 1 0 0 0>}
 ]
}
